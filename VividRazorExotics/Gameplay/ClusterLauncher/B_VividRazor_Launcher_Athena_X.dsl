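; Weapon Blueprint for the VividRazor cluster launcher, driving its four-muzzle firing effects and reload magazine-eject effect.
; Fire-start and local reload-start events use DoN counters to select and activate front and back muzzle effects in turn.
; At the matching reload stage, a delay precedes magazine particle activation.
; Particle-collision sound handling only prints debug text, and the projectile-spawned handler is empty.
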(event Custom|SetMuzzles
  (Default|BindEventtoOnStartFiring (delegate Custom|OnStartFiring_Event)))

(event Custom|OnStartFiring_Event
  (bind _counter (Utilities|FlowControl|DoN 4)
    (:Exit
      (Components|Activation|Activate (Utilities|Select :Index _counter :Option%201 (Variables|Muzzle|GetMuzzle1) :Option%202 (Variables|Default|GetMuzzle2) :Option%203 (Variables|Default|GetMuzzle3) :Option%204 (Variables|Default|GetMuzzle4)) true)
      (Components|Activation|Activate (Utilities|Select :Index _counter :Option%201 (Variables|Default|GetNS_MuzzleBack1) :Option%202 (Variables|Default|GetNS_MuzzleBack2) :Option%203 (Variables|Default|GetNS_MuzzleBack3) :Option%204 (Variables|Default|GetNS_MuzzleBack4)) true)
      (Utilities|FlowControl|Delay 0.05))))

(event EventLocalReloadStarted
  (params (ReloadTime :type Float%20%28single-precision%29) (ReloadType :type EFortWeaponReloadType%20Enum :default "ReloadWholeClip"))
  (bind _counter (Utilities|FlowControl|DoN 4)
    (:Exit
      (Components|Activation|Activate (Utilities|Select :Index _counter :Option%201 (Variables|Muzzle|GetMuzzle1) :Option%202 (Variables|Default|GetMuzzle2) :Option%203 (Variables|Default|GetMuzzle3) :Option%204 (Variables|Default|GetMuzzle4)) true)
      (Components|Activation|Activate (Utilities|Select :Index _counter :Option%201 (Variables|Default|GetNS_MuzzleBack1) :Option%202 (Variables|Default|GetNS_MuzzleBack2) :Option%203 (Variables|Default|GetNS_MuzzleBack3) :Option%204 (Variables|Default|GetNS_MuzzleBack4)) true)
      (Utilities|FlowControl|Delay 0.05))))

(event EventBeginPlay
  (EventDispatchers|BindEventtoOnProjectileSpawned (delegate Custom|OnProjectileSpawned_Event))
  (CallFunction|SetMuzzles))

(event Custom|OnProjectileSpawned_Event
  (params (Projectile :type Fort%20Projectile%20Base%20Object%20Reference)))

(event EventMuzzleFlashFX
  (params (Persistent%20Fire :type Boolean))
  (|CallParentFunction Persistent%20Fire))

(event EventOnPlayReloadFX
  (params (ReloadStage :type EFortReloadFXState%20Enum :default "ReloadStart"))
  ;[1 1 1] Magazine ejecting and bouncing animation while reloading
  (EventDispatchers|BindEventtoOnParticleCollide (Variables|Default|GetP_M202_Muzzle_MagazineEject) (delegate Custom|SoundFX))
  (if (and (Utilities|Enum|Equal(Enum) ReloadStage (Variables|Default|GetReloadState)) (FortWeapon|IsReloading))
    (Utilities|FlowControl|Delay 0.55)
    (Development|PrintString (Utilities|String|EnumtoString ReloadStage) true true "(R=0.000000,G=0.660000,B=1.000000,A=1.000000)" 8.0)
    (Components|Activation|Activate (Variables|Default|GetP_M202_Muzzle_MagazineEject) false)))

(event Custom|SoundFX
  (params (EventName :type Name) (EmitterTime :type Float%20%28single-precision%29) (ParticleTime :type Integer) (Location :type Vector) (Velocity :type Vector) (Direction :type Vector) (Normal :type Vector) (BoneName :type Name) (PhysMat :type Physical%20Material%20Object%20Reference))
  ;[1 1 1] Bound a particle collision event For SFX
  (Utilities|FlowControl|DoOnce false
    (:Completed
      (Development|PrintString "ActiveSFX")
      (CallFunction|ResetDoOnce))))

(event Custom|ResetDoOnce
  ;[1 1 1] Bound a particle collision event For SFX
  (Utilities|FlowControl|DoOnce false
    (:Completed
      (Development|PrintString "ActiveSFX")
      (CallFunction|ResetDoOnce))))

(fn ConstructionScript
  (|CallParentFunction))
